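# NS_Pickaxe_Troops_Swing is a Niagara system that draws a ribbon swing trail for the Troops pickaxe melee weapon.
# Its single Wyeth_AnimTrails emitter spawns trail particles from the User.SkeletalMesh parameter via custom anim-trail spawn, initialize and update modules.
# Short-lived, pale semi-transparent particles feed a ribbon renderer, with fixed-seed random dynamic material parameters and a curve that fades alpha over life.
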NiagaraSystem /BRCosmetics/Weapons/FORT_Melee/Pickaxe_Troops/FX/NS_Pickaxe_Troops_Swing
dependencies {usedDataInterfaces [&NiagaraDataInterfaceCurve], usedDynamicInputs [], usedModules [&SystemState]}
emitters {Wyeth_AnimTrails {
  emitterUpdateScript [
   &EmitterState(Life Cycle Mode=::NewEnumerator1, Loop Duration=0.5, UseLoopDelay=false)
   &self:SpawnAnimTrail(Mesh:NiagaraDataInterfaceSkeletalMesh=$User.SkeletalMesh, Socket A=1) {module=SpawnAnimTrail}
  ]
  particleSpawnScript [
   &self:InitializeAnimTrail(Color:LinearColor=<rgba 0.9045 0.9341 1 0.333>, Lifetime=0.175, Ribbon Width Scale=1) {module=InitializeAnimTrail}
   &DynamicMaterialParameters(
    Index 0 Param 1=&RandomRangeFloat(Minimum=0, Maximum=1, Randomness Mode=::NewEnumerator1, Fixed Random Seed=true)
    Index 0 Param 2=&RandomRangeFloat(Minimum=0, Maximum=1, Randomness Mode=::NewEnumerator1, Random Seed=1, Fixed Random Seed=true)
    Index 0 Param 3=&RandomRangeFloat(Minimum=0, Maximum=1, Randomness Mode=::NewEnumerator1, Random Seed=2, Fixed Random Seed=true)
    Module.Param0WriteEnabled=true
    Module.Param1WriteEnabled=true
    Module.Param2WriteEnabled=true
    Module.Param3WriteEnabled=true
   )
  ]
  particleUpdateScript [
   &ParticleState() {module=UpdateAge}
   &SolveForcesAndVelocity(Clamp Velocity=false, Limit Acceleration=false)
   &self:UpdateAnimTrail() {module=UpdateAnimTrail}
   &ScaleColor(
    Scale Alpha=&FloatFromCurve(FloatCurve=DataInterface() {propertyValues {Curve {keys [<at 0 1 RCIM_Cubic>, {arriveTangentWeight=0.797, interpMode=RCIM_Cubic, tangentMode=RCTM_User, time=1, value=0}]}}})
    ScaleA=true
    ScaleRGB=true
   )
  ]
  rendererClasses [&NiagaraRibbonRendererProperties]
  renderers [{rendererClass=&NiagaraRibbonRendererProperties}]
 }}
summary {
 systemName=NS_Pickaxe_Troops_Swing
 emitters [{emitterName=Wyeth_AnimTrails, rendererClasses [&NiagaraRibbonRendererProperties]}]
 userVariables [
  User.bisFrontEndPreview {type=NiagaraBool}
  User.bisFullyDeployed {type=NiagaraBool}
  User.ForwardDot {type=NiagaraFloat, defaultValue=0, description=""}
  User.RibbonIDOffset {type=NiagaraInt32}
  User.SkeletalMesh {
   type=NiagaraDataInterfaceSkeletalMesh
   defaultValue {
    struct=NiagaraExt_VariableValue_DataInterface
    value {dataInterface=&self:NiagaraDataInterfaceSkeletalMesh_0}
   }
  }
 ]
}
